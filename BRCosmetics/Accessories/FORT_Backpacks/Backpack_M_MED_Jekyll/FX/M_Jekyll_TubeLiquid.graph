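# Masked, two-sided material for the liquid in the Jekyll back bling's tubes, built for Niagara mesh particles.
# SplineThicken shapes strand width from dynamic base and tip parameters.
# Color blends from darkened to full particle color along a sine of U.
# Panning noise distorts a soft radial ramp shifted by LiquidMotion, which is cut off by LiquidLength_Max and scaled into a hard-edged mask; LiquidLength_Grow is unused.
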Material /BRCosmetics/Accessories/FORT_Backpacks/Backpack_M_MED_Jekyll/FX/M_Jekyll_TubeLiquid
graph {
 properties {bUsedWithNiagaraMeshParticles=true, blendMode=BLEND_Masked, pixelDepthOffsetMode=PDOM_Legacy, twoSided=true, usedShadingModels=Default Lit}
 nodes [
  n1 = DynamicParameter() {parameterIndex=1, defaultValue=<rgba 1 1 0.1 1>, paramNames [BaseWidth,TipWidth,Param3,Param4]}
  n2 = ScalarParameter() {defaultValue=0.8, parameterName="WidthBase (S)"}
  MaterialFunctionCall(WidthBase=Multiply(A=@n1.BaseWidth, B=@n2), WidthTip=Multiply(A=@n2, B=@n1.TipWidth)) {materialFunction=&SplineThicken}
  n3 = VectorParameter() {parameterName=Dark_Color, defaultValue=<rgba 0.3 0.3 0.3 1>}
  n4 = ParticleColor()
  n5 = Multiply(A=ComponentMask(None=TextureCoordinate()) {r=true}) {constB=0.5}
  MaterialFunctionCall(Emissive=LinearInterpolate(A=Multiply(A=@n3.RGB, B=@n4.RGB), B=@n4.RGB, Alpha=Sine(None=@n5))) {materialFunction=&MF_FX_EmissiveAdjustments}
  n6 = TextureCoordinate()
  n7 = Constant2Vector() {g=1, r=0.2}
  n8 = TextureSample(UVs=Panner(Coordinate=Multiply(A=@n6, B=@n7)) {speedY=0.5}) {samplerType=SAMPLERTYPE_Grayscale, texture=&LowResBlurredNoise}
  n9 = ComponentMask(None=@n6) {r=true}
  n10 = DynamicParameter() {defaultValue=<rgba 1 0.01 0.1 1>, paramNames [LiquidLength_Grow, LiquidLength_Max, LiquidMotion, Param4]}
  n11 = Clamp(None=Add(A=ComponentMask(None=@n6) {g=true}, B=@n10.LiquidMotion))
  n12 = If(A=@n11, B=Constant() {r=1}, "A > B"=Constant(), "A < B"=@n11)
  n13 = TextureSample(UVs=Add(A=Multiply(A=@n8.R) {constB=0.1}, B=AppendVector(A=@n9, B=@n12))) {texture=/Game/Effects/Fort_Effects/Textures/Ramps/T_Ramp_Radial_Soft}
  Multiply(A=If(A=@n13.R, B=LinearInterpolate(B=@n10.LiquidLength_Max, Alpha=@n10.LiquidLength_Max), "A > B"=Constant(), "A == B"=@n13.R, "A < B"=@n13.R)) {constB=100}
 ]
}
