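# Physics asset for the Jekyll backpack cosmetic, apparently for secondary motion of its liquid and flask-liquid bones.
# It has three box bodies: a small one on the pack bone and flat slabs on the two dyn_ upper liquid bones.
# Each liquid body is constrained directly to the pack, with matching tight, asymmetric swing and twist limit angles that suggest a restrained slosh.
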PhysicsAsset /BRCosmetics/Accessories/FORT_Backpacks/Backpack_M_MED_Jekyll/Meshes/M_MED_Jekyll_Pack_Physics
bodies {
 dyn_flask_liquid_upper [{shapeType=Box, extentX=1, extentY=7, extentZ=7, center=<0.08 0 0>}]
 dyn_liquid_upper [{shapeType=Box, extentX=1, extentY=4, extentZ=4, center=<0.08 0 0>}]
 pack [{shapeType=Box, extentX=1, extentY=1, extentZ=1}]
}
constraints [
 {bone1Name=dyn_flask_liquid_upper, bone2Name=pack, swing1LimitDegrees=15, swing2LimitDegrees=5, twistLimitDegrees=2}
 {bone1Name=dyn_liquid_upper, bone2Name=pack, swing1LimitDegrees=15, swing2LimitDegrees=5, twistLimitDegrees=2}
]
