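# Animation Blueprint for the Jekyll backpack cosmetic, playing a repeating lab sequence of flasks filling, tipping, pouring and draining.
# A self-resetting DoOnce chain of Delays steps blend floats and fires the ActiveLoopingFX dispatcher each cycle.
# Those floats drive interpolated blends between fixed bone-transform poses that fake liquid levels by scaling bones.
# RigidBody simulation feeds only the base pose; sine-driven pot wobble and needle spin appear only in the empty-flask return pose.
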AnimBlueprint /BRCosmetics/Accessories/FORT_Backpacks/Backpack_M_MED_Jekyll/Meshes/M_MED_Jekyll_Pack_AnimBP
target_skeleton=&/M_MED_Jekyll_Pack_Skeleton
graphs [
 ActiveLoopingFX {nodes [ActiveLoopingFX() {is=title K2Node_FunctionEntry}]}
 AnimGraph {nodes [
   "Transform (Modify) Bone - Bone: dyn_liquid"(Alpha=1, Translation="0,0,0", Rotation="0,0,0", Scale="0,0,0") {
    rotationMode=BMM_Additive
    rotationSpace=BCS_BoneSpace
    alphaBoolBlend {blendOption=Linear}
    boneToModify {boneName=dyn_liquid}
    scale=<0 0 0>
   }
   n1 = "Transform (Modify) Bone - Bone: dyn_liquid_upper"(Alpha=1, Translation="5.1,0,0", Rotation="0,0,0", Scale="0,0,0") {
    rotationSpace=BCS_BoneSpace
    scaleMode=BMM_Replace
    scaleSpace=BCS_BoneSpace
    translationMode=BMM_Replace
    translationSpace=BCS_BoneSpace
    alphaBoolBlend {blendOption=Linear}
    boneToModify {boneName=dyn_liquid_upper}
    scale=<0 0 0>
    translation {x=5.1}
   }
   n2 = "Transform (Modify) Bone - Bone: dyn_liquid"(ComponentPose=@n1, Alpha=1, Translation="0,0,0", Rotation="0,0,0", Scale="1,1,1") {rotationMode=BMM_Additive, rotationSpace=BCS_BoneSpace, alphaBoolBlend {blendOption=Linear}, boneToModify {boneName=dyn_liquid}}
   n3 = "Transform (Modify) Bone - Bone: dyn_liquid_lower"(ComponentPose=@n2, Alpha=1, Translation="0,0,0", Rotation="0,0,0", Scale="0,0,0") {
    rotationSpace=BCS_BoneSpace
    scaleMode=BMM_Replace
    scaleSpace=BCS_BoneSpace
    translationMode=BMM_Replace
    translationSpace=BCS_BoneSpace
    alphaBoolBlend {blendOption=Linear}
    boneToModify {boneName=dyn_liquid_lower}
    scale=<0 0 0>
   }
   n4 = "Transform (Modify) Bone - Bone: dyn_master_flask"(ComponentPose=@n3, Alpha=1, Translation="0,0,0", Rotation="0,0,0", Scale="1,1,1") {rotationMode=BMM_Additive, rotationSpace=BCS_BoneSpace, alphaBoolBlend {blendOption=Linear}, boneToModify {boneName=dyn_master_flask}}
   n5 = "Transform (Modify) Bone - Bone: dyn_flask_liquid_upper"(ComponentPose=@n4, Alpha=1, Translation="0,0,0", Rotation="0,0,0", Scale="1,1,1") {
    rotationSpace=BCS_BoneSpace
    scaleMode=BMM_Replace
    scaleSpace=BCS_BoneSpace
    translationMode=BMM_Replace
    translationSpace=BCS_BoneSpace
    alphaBoolBlend {blendOption=Linear}
    boneToModify {boneName=dyn_flask_liquid_upper}
   }
   n6 = "Transform (Modify) Bone - Bone: dyn_flask_liquid_lower"(ComponentPose=@n5, Alpha=1, Translation="0,0,0", Rotation="0,0,0", Scale="1,1,1") {
    rotationSpace=BCS_BoneSpace
    scaleMode=BMM_Replace
    scaleSpace=BCS_BoneSpace
    translationMode=BMM_Replace
    translationSpace=BCS_BoneSpace
    alphaBoolBlend {blendOption=Linear}
    boneToModify {boneName=dyn_flask_liquid_lower}
   }
   n7 = "Transform (Modify) Bone - Bone: dyn_liquid_upper"(Alpha=1, Translation="5.1,0,0", Rotation="0,0,0", Scale="0,0,0") {
    rotationSpace=BCS_BoneSpace
    scaleMode=BMM_Replace
    scaleSpace=BCS_BoneSpace
    translationMode=BMM_Replace
    translationSpace=BCS_BoneSpace
    alphaBoolBlend {blendOption=Linear}
    boneToModify {boneName=dyn_liquid_upper}
    scale=<0 0 0>
    translation {x=5.1}
   }
   n8 = "Transform (Modify) Bone - Bone: dyn_liquid"(ComponentPose=@n7, Alpha=1, Translation="0,0,0", Rotation="0,0,0", Scale="1,1,1") {rotationMode=BMM_Additive, rotationSpace=BCS_BoneSpace, alphaBoolBlend {blendOption=Linear}, boneToModify {boneName=dyn_liquid}}
   n9 = "Transform (Modify) Bone - Bone: dyn_liquid_lower"(ComponentPose=@n8, Alpha=1, Translation="0,0,0", Rotation="0,0,0", Scale="0,0,0") {
    rotationSpace=BCS_BoneSpace
    scaleMode=BMM_Replace
    scaleSpace=BCS_BoneSpace
    translationMode=BMM_Replace
    translationSpace=BCS_BoneSpace
    alphaBoolBlend {blendOption=Linear}
    boneToModify {boneName=dyn_liquid_lower}
    scale=<0 0 0>
   }
   n10 = "Transform (Modify) Bone - Bone: dyn_master_flask"(ComponentPose=@n9, Alpha=1, Translation="0,0,0", Rotation="0,0,0", Scale="1,1,1") {rotationMode=BMM_Additive, rotationSpace=BCS_BoneSpace, alphaBoolBlend {blendOption=Linear}, boneToModify {boneName=dyn_master_flask}}
   n11 = "Transform (Modify) Bone - Bone: dyn_flask_liquid_upper"(ComponentPose=@n10, Alpha=1, Translation="5.4,0,0", Rotation="0,0,0", Scale="0,1,1") {
    rotationSpace=BCS_BoneSpace
    scaleMode=BMM_Replace
    scaleSpace=BCS_BoneSpace
    translationMode=BMM_Replace
    translationSpace=BCS_BoneSpace
    alphaBoolBlend {blendOption=Linear}
    boneToModify {boneName=dyn_flask_liquid_upper}
    scale {x=0}
    translation {x=5.4}
   }
   n12 = "Transform (Modify) Bone - Bone: dyn_flask_liquid_lower"(ComponentPose=@n11, Alpha=1, Translation="0,0,0", Rotation="0,0,0", Scale="0,0,0") {
    rotationSpace=BCS_BoneSpace
    scaleMode=BMM_Replace
    scaleSpace=BCS_BoneSpace
    translationMode=BMM_Replace
    translationSpace=BCS_BoneSpace
    alphaBoolBlend {blendOption=Linear}
    boneToModify {boneName=dyn_flask_liquid_lower}
    scale=<0 0 0>
   }
   n13 = "Transform (Modify) Bone - Bone: dyn_liquid_upper"(Alpha=1, Translation="5.4,0,0", Rotation="0,0,0", Scale="0,0,0") {
    rotationSpace=BCS_BoneSpace
    scaleMode=BMM_Replace
    scaleSpace=BCS_BoneSpace
    translationMode=BMM_Replace
    translationSpace=BCS_BoneSpace
    alphaBoolBlend {blendOption=Linear}
    boneToModify {boneName=dyn_liquid_upper}
    scale=<0 0 0>
    translation {x=5.4}
   }
   n14 = "Transform (Modify) Bone - Bone: dyn_liquid"(ComponentPose=@n13, Alpha=1, Translation="0,0,0", Rotation="0,0,0", Scale="1,1,1") {rotationMode=BMM_Additive, rotationSpace=BCS_BoneSpace, alphaBoolBlend {blendOption=Linear}, boneToModify {boneName=dyn_liquid}}
   n15 = "Transform (Modify) Bone - Bone: dyn_liquid_lower"(ComponentPose=@n14, Alpha=1, Translation="0,0,0", Rotation="0,0,0", Scale="0,0,0") {
    rotationSpace=BCS_BoneSpace
    scaleMode=BMM_Replace
    scaleSpace=BCS_BoneSpace
    translationMode=BMM_Replace
    translationSpace=BCS_BoneSpace
    alphaBoolBlend {blendOption=Linear}
    boneToModify {boneName=dyn_liquid_lower}
    scale=<0 0 0>
   }
   n16 = "Transform (Modify) Bone - Bone: dyn_master_flask"(ComponentPose=@n15, Alpha=1, Translation="0,0,0", Rotation="0,0,0", Scale="1,1,1") {rotationMode=BMM_Additive, rotationSpace=BCS_BoneSpace, alphaBoolBlend {blendOption=Linear}, boneToModify {boneName=dyn_master_flask}}
   n17 = "Transform (Modify) Bone - Bone: dyn_flask_liquid_upper"(ComponentPose=@n16, Alpha=1, Translation="5.4,0,0", Rotation="0,0,0", Scale="0,0,0") {
    rotationSpace=BCS_BoneSpace
    scaleMode=BMM_Replace
    scaleSpace=BCS_BoneSpace
    translationMode=BMM_Replace
    translationSpace=BCS_BoneSpace
    alphaBoolBlend {blendOption=Linear}
    boneToModify {boneName=dyn_flask_liquid_upper}
    scale=<0 0 0>
    translation {x=5.4}
   }
   n18 = "Transform (Modify) Bone - Bone: dyn_flask_liquid_lower"(ComponentPose=@n17, Alpha=1, Translation="0,0,0", Rotation="0,0,0", Scale="0,0,0") {
    rotationSpace=BCS_BoneSpace
    scaleMode=BMM_Replace
    scaleSpace=BCS_BoneSpace
    translationMode=BMM_Replace
    translationSpace=BCS_BoneSpace
    alphaBoolBlend {blendOption=Linear}
    boneToModify {boneName=dyn_flask_liquid_lower}
    scale=<0 0 0>
   }
   n19 = "Transform (Modify) Bone - Bone: dyn_master_flask"(ComponentPose=@n18, Alpha=1, Translation="0,0,0", Rotation="0,0,0", Scale="1,1,1") {rotationMode=BMM_Additive, rotationSpace=BCS_BoneSpace, alphaBoolBlend {blendOption=Linear}, boneToModify {boneName=dyn_master_flask}}
   n20 = "Transform (Modify) Bone - Bone: dyn_liquid"(ComponentPose=@n19, Alpha=1, Translation="0,0,0", Rotation="0,0,0", Scale="1,1,1") {rotationMode=BMM_Additive, rotationSpace=BCS_BoneSpace, alphaBoolBlend {blendOption=Linear}, boneToModify {boneName=dyn_liquid}}
   n21 = "Transform (Modify) Bone - Bone: dyn_pot"(ComponentPose=@n20, Alpha=1, Translation="0,0,0", Rotation=Get Pot Side to Side() -> (Pot Side to Side:Rotator), Scale="1,1,1") {rotationMode=BMM_Additive, rotationSpace=BCS_BoneSpace, alphaBoolBlend {blendOption=Linear}, boneToModify {boneName=dyn_pot}}
   n22 = "Transform (Modify) Bone - Bone: dyn_spinner"(ComponentPose=@n21, Alpha=1, Translation="0,0,0", Rotation=Get Needle Spinning() -> (Needle Spinning:Rotator), Scale="1,1,1") {rotationMode=BMM_Additive, rotationSpace=BCS_BoneSpace, alphaBoolBlend {blendOption=Linear}, boneToModify {boneName=dyn_spinner}}
   n23 = "Transform (Modify) Bone - Bone: dyn_liquid_upper"(Alpha=1, Translation="5.4,0,0", Rotation="0,0,0", Scale="0.05,0.8,0.8") {
    rotationSpace=BCS_BoneSpace
    scaleMode=BMM_Replace
    scaleSpace=BCS_BoneSpace
    translationMode=BMM_Replace
    translationSpace=BCS_BoneSpace
    alphaBoolBlend {blendOption=Linear}
    boneToModify {boneName=dyn_liquid_upper}
    scale=<0.05 0.8 0.8>
    translation {x=5.4}
   }
   n24 = "Transform (Modify) Bone - Bone: dyn_liquid"(ComponentPose=@n23, Alpha=1, Translation="0,0,0", Rotation="0,0,0", Scale="1,1,1") {rotationMode=BMM_Additive, rotationSpace=BCS_BoneSpace, alphaBoolBlend {blendOption=Linear}, boneToModify {boneName=dyn_liquid}}
   n25 = "Transform (Modify) Bone - Bone: dyn_liquid_lower"(ComponentPose=@n24, Alpha=1, Translation="0,0,0", Rotation="0,0,0", Scale="1,1,1") {
    rotationSpace=BCS_BoneSpace
    scaleMode=BMM_Replace
    scaleSpace=BCS_BoneSpace
    translationMode=BMM_Replace
    translationSpace=BCS_BoneSpace
    alphaBoolBlend {blendOption=Linear}
    boneToModify {boneName=dyn_liquid_lower}
   }
   n26 = "Transform (Modify) Bone - Bone: dyn_master_flask"(ComponentPose=@n25, Alpha=1, Translation="0,0,0", Rotation="0,0,0", Scale="1,1,1") {rotationMode=BMM_Additive, rotationSpace=BCS_BoneSpace, alphaBoolBlend {blendOption=Linear}, boneToModify {boneName=dyn_master_flask}}
   n27 = "Transform (Modify) Bone - Bone: dyn_flask_liquid_upper"(ComponentPose=@n26, Alpha=1, Translation="5.4,0,0", Rotation="0,0,0", Scale="0,0,0") {
    rotationSpace=BCS_BoneSpace
    scaleMode=BMM_Replace
    scaleSpace=BCS_BoneSpace
    translationMode=BMM_Replace
    translationSpace=BCS_BoneSpace
    alphaBoolBlend {blendOption=Linear}
    boneToModify {boneName=dyn_flask_liquid_upper}
    scale=<0 0 0>
    translation {x=5.4}
   }
   n28 = "Transform (Modify) Bone - Bone: dyn_flask_liquid_lower"(ComponentPose=@n27, Alpha=1, Translation="0,0,0", Rotation="0,0,0", Scale="0,0,0") {
    rotationSpace=BCS_BoneSpace
    scaleMode=BMM_Replace
    scaleSpace=BCS_BoneSpace
    translationMode=BMM_Replace
    translationSpace=BCS_BoneSpace
    alphaBoolBlend {blendOption=Linear}
    boneToModify {boneName=dyn_flask_liquid_lower}
    scale=<0 0 0>
   }
   n29 = "Transform (Modify) Bone - Bone: dyn_master_flask"(ComponentPose=@n28, Alpha=1, Translation="0,0,0", Rotation="0,0,-120", Scale="1,1,1") {
    rotationMode=BMM_Additive
    rotationSpace=BCS_BoneSpace
    alphaBoolBlend {blendOption=Linear}
    boneToModify {boneName=dyn_master_flask}
    rotation {roll=-120}
   }
   n30 = "Transform (Modify) Bone - Bone: dyn_liquid"(ComponentPose=@n29, Alpha=1, Translation="0,0,0", Rotation="-120,0,0", Scale="1,1,1") {
    rotationMode=BMM_Additive
    rotationSpace=BCS_BoneSpace
    alphaBoolBlend {blendOption=Linear}
    boneToModify {boneName=dyn_liquid}
    rotation {pitch=-120}
   }
   n31 = Get Liquid drain upper flask() -> (Liquid drain upper flask:double)
   n32 = "Transform (Modify) Bone - Bone: dyn_liquid_upper"(ComponentPose=@n30, Alpha=@n31, Translation="5.4,0,0", Rotation="0,0,0", Scale="0,0,0") {
    rotationSpace=BCS_BoneSpace
    scaleMode=BMM_Replace
    scaleSpace=BCS_BoneSpace
    translationMode=BMM_Replace
    translationSpace=BCS_BoneSpace
    alphaBoolBlend {blendOption=Linear}
    alphaScaleBiasClamp {bInterpResult=true, interpSpeedDecreasing=2, interpSpeedIncreasing=2}
    boneToModify {boneName=dyn_liquid_upper}
    scale=<0 0 0>
    translation {x=5.4}
   }
   n33 = "Transform (Modify) Bone - Bone: dyn_liquid"(ComponentPose=@n32, Alpha=@n31, Translation="0,0,0", Rotation="0,0,0", Scale="1,1,1") {
    rotationMode=BMM_Additive
    rotationSpace=BCS_BoneSpace
    alphaBoolBlend {blendOption=Linear}
    alphaScaleBiasClamp {bInterpResult=true, interpSpeedDecreasing=2, interpSpeedIncreasing=2}
    boneToModify {boneName=dyn_liquid}
   }
   n34 = "Transform (Modify) Bone - Bone: dyn_liquid_lower"(ComponentPose=@n33, Alpha=@n31, Translation="0,0,0", Rotation="0,0,0", Scale="0,0,0") {
    rotationSpace=BCS_BoneSpace
    scaleMode=BMM_Replace
    scaleSpace=BCS_BoneSpace
    translationMode=BMM_Replace
    translationSpace=BCS_BoneSpace
    alphaBoolBlend {blendOption=Linear}
    alphaScaleBiasClamp {bInterpResult=true, interpSpeedDecreasing=2, interpSpeedIncreasing=2}
    boneToModify {boneName=dyn_liquid_lower}
    scale=<0 0 0>
   }
   n35 = "Transform (Modify) Bone - Bone: dyn_master_flask"(ComponentPose=@n34, Alpha=@n31, Translation="0,0,0", Rotation="0,0,0", Scale="1,1,1") {
    rotationMode=BMM_Additive
    rotationSpace=BCS_BoneSpace
    alphaBoolBlend {blendOption=Linear}
    alphaScaleBiasClamp {bInterpResult=true, interpSpeedDecreasing=2, interpSpeedIncreasing=2}
    boneToModify {boneName=dyn_master_flask}
   }
   n36 = "Transform (Modify) Bone - Bone: dyn_liquid_upper"(Alpha=1, Translation="0,0,0", Rotation="0,0,0", Scale="0.9,1,1") {
    rotationSpace=BCS_BoneSpace
    scaleMode=BMM_Replace
    scaleSpace=BCS_BoneSpace
    translationMode=BMM_Replace
    translationSpace=BCS_BoneSpace
    alphaBoolBlend {blendOption=Linear}
    boneToModify {boneName=dyn_liquid_upper}
    scale {x=0.9}
   }
   n37 = "Transform (Modify) Bone - Bone: dyn_liquid"(ComponentPose=@n36, Alpha=1, Translation="0,0,0", Rotation="0,0,0", Scale="1,1,1") {rotationMode=BMM_Additive, rotationSpace=BCS_BoneSpace, alphaBoolBlend {blendOption=Linear}, boneToModify {boneName=dyn_liquid}}
   n38 = "Transform (Modify) Bone - Bone: dyn_liquid_lower"(ComponentPose=@n37, Alpha=1, Translation="0,0,0", Rotation="0,0,0", Scale="1,1,1") {
    rotationSpace=BCS_BoneSpace
    scaleMode=BMM_Replace
    scaleSpace=BCS_BoneSpace
    translationMode=BMM_Replace
    translationSpace=BCS_BoneSpace
    alphaBoolBlend {blendOption=Linear}
    boneToModify {boneName=dyn_liquid_lower}
   }
   n39 = "Transform (Modify) Bone - Bone: dyn_master_flask"(ComponentPose=@n38, Alpha=1, Translation="0,0,0", Rotation="0,0,0", Scale="1,1,1") {rotationMode=BMM_Additive, rotationSpace=BCS_BoneSpace, alphaBoolBlend {blendOption=Linear}, boneToModify {boneName=dyn_master_flask}}
   n40 = "Transform (Modify) Bone - Bone: dyn_flask_liquid_upper"(ComponentPose=@n39, Alpha=1, Translation="5.4,0,0", Rotation="0,0,0", Scale="0,0,0") {
    rotationSpace=BCS_BoneSpace
    scaleMode=BMM_Replace
    scaleSpace=BCS_BoneSpace
    translationMode=BMM_Replace
    translationSpace=BCS_BoneSpace
    alphaBoolBlend {blendOption=Linear}
    boneToModify {boneName=dyn_flask_liquid_upper}
    scale=<0 0 0>
    translation {x=5.4}
   }
   n41 = "Transform (Modify) Bone - Bone: dyn_flask_liquid_lower"(ComponentPose=@n40, Alpha=1, Translation="0,0,0", Rotation="0,0,0", Scale="0,0,0") {
    rotationSpace=BCS_BoneSpace
    scaleMode=BMM_Replace
    scaleSpace=BCS_BoneSpace
    translationMode=BMM_Replace
    translationSpace=BCS_BoneSpace
    alphaBoolBlend {blendOption=Linear}
    boneToModify {boneName=dyn_flask_liquid_lower}
    scale=<0 0 0>
   }
   n42 = "Transform (Modify) Bone - Bone: dyn_master_flask"(ComponentPose=@n41, Alpha=1, Translation="0,0,0", Rotation="0,0,-120", Scale="1,1,1") {
    rotationMode=BMM_Additive
    rotationSpace=BCS_BoneSpace
    alphaBoolBlend {blendOption=Linear}
    boneToModify {boneName=dyn_master_flask}
    rotation {roll=-120}
   }
   n43 = "Transform (Modify) Bone - Bone: dyn_liquid"(ComponentPose=@n42, Alpha=1, Translation="0,0,0", Rotation="-120,0,0", Scale="1,1,1") {
    rotationMode=BMM_Additive
    rotationSpace=BCS_BoneSpace
    alphaBoolBlend {blendOption=Linear}
    boneToModify {boneName=dyn_liquid}
    rotation {pitch=-120}
   }
   n44 = "Transform (Modify) Bone - Bone: dyn_liquid_upper"(Alpha=1, Translation="0,0,0", Rotation="0,0,0", Scale="1,1,1") {
    rotationSpace=BCS_BoneSpace
    scaleMode=BMM_Replace
    scaleSpace=BCS_BoneSpace
    translationMode=BMM_Replace
    translationSpace=BCS_BoneSpace
    alphaBoolBlend {blendOption=Linear}
    boneToModify {boneName=dyn_liquid_upper}
   }
   n45 = "Transform (Modify) Bone - Bone: dyn_liquid"(ComponentPose=@n44, Alpha=1, Translation="0,0,0", Rotation="0,0,0", Scale="1,1,1") {rotationMode=BMM_Additive, rotationSpace=BCS_BoneSpace, alphaBoolBlend {blendOption=Linear}, boneToModify {boneName=dyn_liquid}}
   n46 = "Transform (Modify) Bone - Bone: dyn_liquid_lower"(ComponentPose=@n45, Alpha=1, Translation="0,0,0", Rotation="0,0,0", Scale="1,1,1") {
    rotationSpace=BCS_BoneSpace
    scaleMode=BMM_Replace
    scaleSpace=BCS_BoneSpace
    translationMode=BMM_Replace
    translationSpace=BCS_BoneSpace
    alphaBoolBlend {blendOption=Linear}
    boneToModify {boneName=dyn_liquid_lower}
   }
   n47 = "Transform (Modify) Bone - Bone: dyn_master_flask"(ComponentPose=@n46, Alpha=1, Translation="0,0,0", Rotation="0,0,0", Scale="1,1,1") {rotationMode=BMM_Additive, rotationSpace=BCS_BoneSpace, alphaBoolBlend {blendOption=Linear}, boneToModify {boneName=dyn_master_flask}}
   n48 = "Transform (Modify) Bone - Bone: dyn_flask_liquid_upper"(ComponentPose=@n47, Alpha=1, Translation="5.4,0,0", Rotation="0,0,0", Scale="0,0,0") {
    rotationSpace=BCS_BoneSpace
    scaleMode=BMM_Replace
    scaleSpace=BCS_BoneSpace
    translationMode=BMM_Replace
    translationSpace=BCS_BoneSpace
    alphaBoolBlend {blendOption=Linear}
    boneToModify {boneName=dyn_flask_liquid_upper}
    scale=<0 0 0>
    translation {x=5.4}
   }
   n49 = "Transform (Modify) Bone - Bone: dyn_flask_liquid_lower"(ComponentPose=@n48, Alpha=1, Translation="0,0,0", Rotation="0,0,0", Scale="0,0,0") {
    rotationSpace=BCS_BoneSpace
    scaleMode=BMM_Replace
    scaleSpace=BCS_BoneSpace
    translationMode=BMM_Replace
    translationSpace=BCS_BoneSpace
    alphaBoolBlend {blendOption=Linear}
    boneToModify {boneName=dyn_flask_liquid_lower}
    scale=<0 0 0>
   }
   n50 = RigidBody(ComponentPose=Local To Component(LocalPose=Local Space Ref Pose())) {
    bClampLinearTranslationLimitToRefPose=true
    bOverrideWorldGravity=true
    overridePhysicsAsset=&/M_MED_Jekyll_Pack_Physics
    simulationSpace=BaseBoneSpace
    alphaBoolBlend {blendOption=Linear}
    baseBoneRef {boneName=root}
    componentAppliedLinearAccClamp=<6000 6000 6000>
    componentLinearAccScale=<0.25 0.25 0.25>
    componentLinearVelScale=<0.25 0.25 0.25>
    overrideWorldGravity {z=-980}
   }
   n51 = "Transform (Modify) Bone - Bone: dyn_liquid_upper"(ComponentPose=@n50, Alpha=1, Translation="5.1,0,0", Rotation="0,0,0", Scale="0,0,0") {
    rotationSpace=BCS_BoneSpace
    scaleMode=BMM_Replace
    scaleSpace=BCS_BoneSpace
    translationMode=BMM_Replace
    translationSpace=BCS_BoneSpace
    alphaBoolBlend {blendOption=Linear}
    boneToModify {boneName=dyn_liquid_upper}
    scale=<0 0 0>
    translation {x=5.1}
   }
   n52 = "Transform (Modify) Bone - Bone: dyn_liquid"(ComponentPose=@n51, Alpha=1, Translation="0,0,0", Rotation="0,0,0", Scale="0,0,0") {
    rotationMode=BMM_Additive
    rotationSpace=BCS_BoneSpace
    alphaBoolBlend {blendOption=Linear}
    boneToModify {boneName=dyn_liquid}
    scale=<0 0 0>
   }
   n53 = "Transform (Modify) Bone - Bone: dyn_liquid_lower"(ComponentPose=@n52, Alpha=1, Translation="0,0,0", Rotation="0,0,0", Scale="0,0,0") {
    rotationSpace=BCS_BoneSpace
    scaleMode=BMM_Replace
    scaleSpace=BCS_BoneSpace
    translationMode=BMM_Replace
    translationSpace=BCS_BoneSpace
    alphaBoolBlend {blendOption=Linear}
    boneToModify {boneName=dyn_liquid_lower}
    scale=<0 0 0>
   }
   n54 = "Transform (Modify) Bone - Bone: dyn_master_flask"(ComponentPose=@n53, Alpha=1, Translation="0,0,0", Rotation="0,0,0", Scale="1,1,1") {
    rotationMode=BMM_Additive
    rotationSpace=BCS_BoneSpace
    scaleSpace=BCS_BoneSpace
    alphaBoolBlend {blendOption=Linear}
    boneToModify {boneName=dyn_master_flask}
   }
   n55 = "Transform (Modify) Bone - Bone: dyn_flask_liquid_upper"(ComponentPose=@n54, Alpha=1, Translation="0,0,0", Rotation="0,0,0", Scale="1,1,1") {
    rotationSpace=BCS_BoneSpace
    scaleMode=BMM_Replace
    scaleSpace=BCS_BoneSpace
    translationMode=BMM_Replace
    translationSpace=BCS_BoneSpace
    alphaBoolBlend {blendOption=Linear}
    boneToModify {boneName=dyn_flask_liquid_upper}
   }
   n56 = "Transform (Modify) Bone - Bone: dyn_flask_liquid_lower"(ComponentPose=@n55, Alpha=1, Translation="0,0,0", Rotation="0,0,0", Scale="1,1,1") {
    rotationSpace=BCS_BoneSpace
    scaleMode=BMM_Replace
    scaleSpace=BCS_BoneSpace
    translationMode=BMM_Replace
    translationSpace=BCS_BoneSpace
    alphaBoolBlend {blendOption=Linear}
    boneToModify {boneName=dyn_flask_liquid_lower}
   }
   n57 = "Transform (Modify) Bone - Bone: dyn_liquid_upper"(Alpha=1, Translation="5.1,0,0", Rotation="0,0,0", Scale="0,0.65,0.65") {
    rotationSpace=BCS_BoneSpace
    scaleMode=BMM_Replace
    scaleSpace=BCS_BoneSpace
    translationMode=BMM_Replace
    translationSpace=BCS_BoneSpace
    alphaBoolBlend {blendOption=Linear}
    boneToModify {boneName=dyn_liquid_upper}
    scale=<0 0.65 0.65>
    translation {x=5.1}
   }
   n58 = "Transform (Modify) Bone - Bone: dyn_liquid_upper"(
    ComponentPose=@n57
    Alpha=Get upper flask liquid scale() -> (upper flask liquid scale:double)
    Translation="0,0,0"
    Rotation="0,0,0"
    Scale="0,0,0"
   ) {
    rotationSpace=BCS_BoneSpace
    scaleMode=BMM_Replace
    scaleSpace=BCS_BoneSpace
    translationSpace=BCS_BoneSpace
    alphaBoolBlend {blendOption=Linear}
    alphaScaleBiasClamp {bInterpResult=true, interpSpeedDecreasing=20, interpSpeedIncreasing=2}
    boneToModify {boneName=dyn_liquid_upper}
    scale=<0 0 0>
   }
   n59 = "Transform (Modify) Bone - Bone: dyn_liquid"(ComponentPose=@n58, Alpha=1, Translation="0,0,0", Rotation="0,0,0", Scale="1,1,1") {rotationMode=BMM_Additive, rotationSpace=BCS_BoneSpace, alphaBoolBlend {blendOption=Linear}, boneToModify {boneName=dyn_liquid}}
   n60 = "Transform (Modify) Bone - Bone: dyn_liquid_lower"(ComponentPose=@n59, Alpha=1, Translation="0,0,0", Rotation="0,0,0", Scale="0,0,0") {
    rotationSpace=BCS_BoneSpace
    scaleMode=BMM_Replace
    scaleSpace=BCS_BoneSpace
    translationMode=BMM_Replace
    translationSpace=BCS_BoneSpace
    alphaBoolBlend {blendOption=Linear}
    boneToModify {boneName=dyn_liquid_lower}
    scale=<0 0 0>
   }
   n61 = "Transform (Modify) Bone - Bone: dyn_master_flask"(ComponentPose=@n60, Alpha=1, Translation="0,0,0", Rotation="0,0,0", Scale="1,1,1") {rotationMode=BMM_Additive, rotationSpace=BCS_BoneSpace, alphaBoolBlend {blendOption=Linear}, boneToModify {boneName=dyn_master_flask}}
   n62 = "Transform (Modify) Bone - Bone: dyn_flask_liquid_upper"(ComponentPose=@n61, Alpha=1, Translation="6.75,0,0", Rotation="0,0,0", Scale="0,0.8,0.8") {
    rotationSpace=BCS_BoneSpace
    scaleMode=BMM_Replace
    scaleSpace=BCS_BoneSpace
    translationMode=BMM_Replace
    translationSpace=BCS_BoneSpace
    alphaBoolBlend {blendOption=Linear}
    boneToModify {boneName=dyn_flask_liquid_upper}
    scale=<0 0.8 0.8>
    translation {x=6.75}
   }
   n63 = "Transform (Modify) Bone - Bone: dyn_flask_liquid_upper"(
    ComponentPose=@n62
    Alpha=Get lower flask liquid scale() -> (lower flask liquid scale:double)
    Translation="0,0,0"
    Rotation="0,0,0"
    Scale="0,0.000,0.000"
   ) {
    rotationSpace=BCS_BoneSpace
    scaleMode=BMM_Replace
    scaleSpace=BCS_BoneSpace
    translationMode=BMM_Replace
    translationSpace=BCS_BoneSpace
    alphaBoolBlend {blendOption=Linear}
    alphaScaleBiasClamp {bInterpResult=true, interpSpeedDecreasing=8, interpSpeedIncreasing=8}
    boneToModify {boneName=dyn_flask_liquid_upper}
    scale=<0 0 0>
   }
   n64 = "Transform (Modify) Bone - Bone: dyn_flask_liquid_lower"(ComponentPose=@n63, Alpha=1, Translation="0,0,0", Rotation="0,0,0", Scale="0,0,0") {
    rotationSpace=BCS_BoneSpace
    scaleMode=BMM_Replace
    scaleSpace=BCS_BoneSpace
    translationMode=BMM_Replace
    translationSpace=BCS_BoneSpace
    alphaBoolBlend {blendOption=Linear}
    boneToModify {boneName=dyn_flask_liquid_lower}
    scale=<0 0 0>
   }
   n65 = Two Way Blend(
    A=Component To Local(ComponentPose=@n56)
    B=Component To Local(ComponentPose=@n64)
    Alpha=Get Lower Flask Liquid Drain() -> (Lower Flask Liquid Drain:double)
   ) {blendNode {alphaBoolBlend {blendOption=Linear}, alphaScaleBiasClamp {bInterpResult=true, interpSpeedDecreasing=30, interpSpeedIncreasing=3.5}}}
   n66 = Two Way Blend(A=@n65, B=Component To Local(ComponentPose=@n49), Alpha=Get Upper Flask Liquid Fill() -> (Upper Flask Liquid Fill:double)) {blendNode {alphaBoolBlend {blendOption=Linear}, alphaScaleBiasClamp {bInterpResult=true, interpSpeedDecreasing=5, interpSpeedIncreasing=3}}}
   n67 = Two Way Blend(
    A=Two Way Blend(A=@n66, B=Component To Local(ComponentPose=@n43), Alpha=Get Upper Flask Rotate() -> (Upper Flask Rotate:double)) {blendNode {alphaBoolBlend {blendOption=Linear}, alphaScaleBiasClamp {bInterpResult=true, interpSpeedIncreasing=2}}}
    B=Component To Local(ComponentPose=@n35)
    Alpha=Get Upper Flask Liquid Pour() -> (Upper Flask Liquid Pour:double)
   ) {blendNode {alphaBoolBlend {blendOption=Linear}, alphaScaleBiasClamp {bInterpResult=true, interpSpeedDecreasing=20, interpSpeedIncreasing=2}}}
   n68 = Two Way Blend(A=@n67, B=Component To Local(ComponentPose=@n22), Alpha=Get Upper Flask Empty Bottle Return() -> (Upper Flask Empty Bottle Return:double)) {blendNode {alphaBoolBlend {blendOption=Linear}, alphaScaleBiasClamp {bInterpResult=true, interpSpeedDecreasing=15, interpSpeedIncreasing=2}}}
   n69 = Two Way Blend(
    A=Two Way Blend(A=@n68, B=Component To Local(ComponentPose=@n12), Alpha=Get Liquid Bottom Fill() -> (Liquid Bottom Fill:double)) {blendNode {alphaBoolBlend {blendOption=Linear}, alphaScaleBiasClamp {bInterpResult=true}}}
    B=Component To Local(ComponentPose=@n6)
    Alpha=Get Return To Start() -> (Return To Start:double)
   ) {blendNode {alphaBoolBlend {blendOption=Linear}, alphaScaleBiasClamp {bInterpResult=true, interpSpeedDecreasing=5, interpSpeedIncreasing=2}}}
   Output Pose(Result=@n69) {layerGroup=None}
   Two Way Blend(B=Component To Local(), Alpha=Get Liquid drain upper flask()) {blendNode {alphaBoolBlend {blendOption=Linear}, alphaScaleBiasClamp {bInterpResult=true, interpSpeedDecreasing=2, interpSpeedIncreasing=2}}}
  ]}
 EventGraph {nodes [
   Event BlueprintUpdateAnimation()
   > n1 = Sequence!(execute) -> (then_0, then_1, then_2:Exec, then_3:Exec)
   n2 = "Sin (Degrees)"!(A:double=float * float(A=GetGameTimeInSeconds(), B=5000)) -> (ReturnValue:double)
   > Set Pot Sway(Pot Sway:Vector=MakeVector(X=float * float(A=@n2, B=1), Y=0)) -> (Output_Get:Vector)
   n3 = "Sin (Degrees)"(A=float * float(A=GetGameTimeInSeconds(), B=5000))
   > Set Pot Side to Side(Pot Side to Side:Rotator=MakeRotator(Roll=float * float(A=@n3, B=1))) -> (Output_Get:Rotator)
   Do Once(execute=@n1.then_1, Reset=@n6.then)
   > Set Start Position(Start Position:double=1) -> (Output_Get:double)
   > Delay(Duration=0.5)
   > Call Active Looping FX!(execute:Exec, self:M MED Jekyll Pack Anim BP Object Reference) -> (then:Exec) {is=title K2Node_CallDelegate}
   > n4 = Sequence!(execute) -> (then_0, then_1)
   > Set Lower Flask Liquid Drain(Lower Flask Liquid Drain:double=1) -> (Output_Get:double)
   > Delay(Duration=2)
   > Set Upper Flask Liquid Fill(Upper Flask Liquid Fill:double=1) -> (Output_Get:double)
   > Delay(Duration=8.5)
   > Set lower flask liquid scale(lower flask liquid scale:double=0) -> (Output_Get:double)
   > Set Upper Flask Rotate(Upper Flask Rotate:double=1) -> (Output_Get:double)
   > Delay(Duration=0.7)
   > Set Upper Flask Liquid Pour(Upper Flask Liquid Pour:double=1) -> (Output_Get:double)
   > n5 = Sequence()
   > Delay(Duration=2)
   > Set Liquid Bottom Fill(Liquid Bottom Fill:double=0) -> (Output_Get:double)
   > Set Upper Flask Empty Bottle Return(Upper Flask Empty Bottle Return:double=1) -> (Output_Get:double)
   > Delay(Duration=3.35)
   > Set Liquid Bottom Fill(Liquid Bottom Fill=1)
   > Set Return To Start(Return To Start:double=1) -> (Output_Get:double)
   > Delay(Duration=3)
   > Set Start Position(Start Position=0)
   > Set Lower Flask Liquid Drain(Lower Flask Liquid Drain=0)
   > Set Liquid Bottom Fill(Liquid Bottom Fill=0)
   > Set Upper Flask Liquid Fill(Upper Flask Liquid Fill=0)
   > Set Upper Flask Rotate(Upper Flask Rotate=0)
   > Set Upper Flask Liquid Pour(Upper Flask Liquid Pour=0)
   > Set Upper Flask Empty Bottle Return(Upper Flask Empty Bottle Return=0)
   > Set Liquid drain upper flask(Liquid drain upper flask:double=0) -> (Output_Get:double)
   > Delay(Duration=5)
   > n6 = Set Return To Start(Return To Start=0)
   Delay(execute=@n5.then_1, Duration=0.35)
   > Set Liquid drain upper flask(Liquid drain upper flask=1)
   Delay(execute=@n4.then_1, Duration=0.385)
   > Set lower flask liquid scale(lower flask liquid scale=1)
   n7 = "Sin (Degrees)"(A=float * float(A=GetGameTimeInSeconds(), B=300))
   Set Needle Spinning(execute=@n1.then_2, Needle Spinning:Rotator=MakeRotator(Roll=float * float(A=@n7, B=180))) -> (Output_Get:Rotator)
   n8 = Get Upper Flask Liquid Fill()
   n9 = Get Lower Flask Liquid Drain()
   Set upper flask liquid scale(execute=@n1.then_3, upper flask liquid scale:double=float - float(B=float * float(A=@n8, B=@n9)) {is=title K2Node_CallFunction}) -> (Output_Get:double)
   TryGetPawnOwner()
   n10 = "Sin (Degrees)"(A=float * float(A=GetGameTimeInSeconds(), B=1))
   Set Needle Spin(Needle Spin:Vector=MakeVector(X=float * float(A=@n10, B=1), Y=0)) -> (Output_Get:Vector)
  ]}
]
variables [
 ActiveLoopingFX {default=()}
 Liquid Bottom Fill {type=number}
 Liquid drain upper flask {type=number}
 Lower Flask Liquid Drain {type=number}
 lower flask liquid scale {type=number}
 Needle Spin {type=Vector}
 Needle Spinning {type=Rotator}
 Pot Side to Side {type=Rotator}
 Pot Sway {type=Vector}
 Return To Start {type=number}
 Start Position {type=number}
 Upper Flask Empty Bottle Return {type=number}
 Upper Flask Liquid Fill {type=number}
 Upper Flask Liquid Pour {type=number}
 upper flask liquid scale {type=number}
 Upper Flask Rotate {type=number}
]
